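# Material function for vehicle body skin Sprinkle04, an animated decal look with spinning, panning "burgers" and trails, output as fully rough, non-metallic color with scalable emissive.
# Base texture gains reflection-vector noise glints and yellow from rotating, masked RGB-texture samples, with some UVs mirrored.
# Repeated Rib-texture samples on UV1 build burger color, alpha and trail masks, shared through named reroutes.
# Anisotropic noise breaks up dark areas, and stepped Fresnel darkens the silhouette.
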MaterialFunction /VehicleCosmetics/Bodies/Sprinkle/Skins/MF_Skin_Sprinkle04
graph {nodes [
  n1 = TextureSample() {samplerType=SAMPLERTYPE_Alpha, texture=/VehicleCosmetics/Bodies/Sprinkle/Textures/Body/T_Sprinkle_Body_DFL}
  n2 = OneMinus(None=@n1.R)
  n3 = Fresnel() {baseReflectFraction=0, exponent=2}
  n4 = ConstantBiasScale(None=@n3) {bias=0.1, scale=5}
  n5 = Multiply(A=@n4) {constB=0.05}
  n6 = ConstantBiasScale(None=ReflectionVectorWS()) {bias=0}
  n7 = AppendVector(
   A=TextureSample(UVs=ComponentMask(None=@n6) {g=true, r=true}) {texture=/VehicleCosmetics/Textures/UniversalTextures/Universal_Noise_47_RGB}.R
   B=Constant() {r=0.6}
  )
  n8 = Multiply(A=OneMinus(None=@n3), B=TextureSample(UVs=Add(A=@n5, B=@n7)) {texture=&Universal_Noise_47_RGB}.B)
  n9 = TextureSample(UVs=Add(A=Multiply(A=@n5) {constB=0.5}, B=@n7)) {texture=&Universal_Noise_47_RGB}
  n10 = Max(
   A=Clamp(None=ConstantBiasScale(None=@n8) {bias=-0.1, scale=50}) {maxDefault=0.05}
   B=Clamp(None=ConstantBiasScale(None=@n9.B) {bias=-0.3, scale=10}) {maxDefault=0.1}
  )
  n11 = TextureSample() {texture=&/T_Sprinkle04_BC}
  n12 = TextureSample() {texture=&/T_Sprinkle04_RGB}
  n13 = TextureCoordinate() {coordinateIndex=1}
  n14 = Constant2Vector() {g=0.4}
  n15 = TextureSample(UVs=Rotator(Coordinate=Add(A=@n13, B=@n14))) {texture=&T_Sprinkle04_RGB}
  n16 = Multiply(A=@n12.A, B=Saturate(None=ConstantBiasScale(None=@n15.G) {bias=-0.2, scale=2}))
  n17 = ComponentMask(None=@n13) {r=true}
  n18 = LinearInterpolate(A=@n17, B=OneMinus(None=@n17), Alpha=Ceil(None=ConstantBiasScale(None=@n17) {bias=-0.5, scale=2}))
  n19 = AppendVector(A=ConstantBiasScale(None=OneMinus(None=@n18)) {bias=-0.5, scale=2}, B=ComponentMask(None=@n13) {g=true})
  n20 = Constant2Vector() {g=0.06}
  n21 = TextureSample(UVs=Rotator(Coordinate=Add(A=@n19, B=@n20)) {speed=0.2}) {texture=&T_Sprinkle04_RGB}
  n22 = Multiply(A=@n12.B, B=Saturate(None=ConstantBiasScale(None=@n21.G) {bias=-0.2, scale=2}))
  n23 = Constant3Vector() {constant=<rgba 0.0111 0.0119 0.0135 0>}
  n24 = MaterialFunctionCall!(
   Texture Coordinates
   Tiling X=Constant() {r=100}
   Tiling Y=Constant() {r=0.1}
   PanSpeedX=Constant() {r=0.0002}
   PanSpeedY=Constant() {r=1}
   RotateSpeed
   Angle
   Chaos=Constant() {r=2}
   ChaosSpeed=Constant() {r=0.2}
  ) {materialFunction=/VehicleCosmetics/Materials/MaterialFunctions/MF_Noisemaker_Anisotropic}
  n25 = Multiply(A=NamedRerouteUsage(), B=Saturate(None=ConstantBiasScale(None=@n24.Aniso4) {bias=-0.09, scale=3}))
  n26 = LinearInterpolate(
   A=LinearInterpolate(A=Add(A=@n10, B=@n11.RGB), B=Constant3Vector() {constant=<rgba 1 1 0 0>}, Alpha=Max(A=@n16, B=@n22))
   B=@n23
   Alpha=Saturate(None=ConstantBiasScale(None=@n25) {bias=-0.2, scale=50})
  )
  n27 = Fresnel() {baseReflectFraction=0, exponent=1}
  n28 = ConstantBiasScale(None=@n27) {bias=-0.75, scale=50}
  n29 = ConstantBiasScale(None=@n27) {bias=-0.6, scale=100}
  n30 = LinearInterpolate(
   A=LinearInterpolate(A=Multiply(A=@n2, B=@n26), B=NamedRerouteUsage(), Alpha=NamedRerouteUsage())
   B=@n23
   Alpha=Max(A=Clamp(None=@n28) {maxDefault=0.75}, B=Clamp(None=@n29) {maxDefault=0.5})
  )
  n31 = SetMaterialAttributes!(
   MaterialAttributes
   BaseColor=@n30
   Metallic=Constant()
   Roughness=Constant() {r=1}
   EmissiveColor=Multiply(A=@n30, B=ScalarParameter() {parameterName=EmissiveIntensity})
   Normal
   AmbientOcclusion
  ) {attributeSetTypes [
    #1
    #2
    #3
    #4
    #5
    #6
   ]}
  FunctionOutput(None=@n31)
  n32 = TextureCoordinate() {uTiling=0.65}
  n33 = ComponentMask(None=@n32) {g=true}
  n34 = ComponentMask(None=@n32) {r=true}
  n35 = OneMinus(None=TextureSample(UVs=AppendVector(A=@n33, B=@n34)) {samplerType=SAMPLERTYPE_Normal, texture=/VehicleCosmetics/Textures/UniversalTextures/Universal_Noise_Normal_30}.G)
  Multiply(A=@n4, B=Saturate(None=ConstantBiasScale(None=@n35) {bias=-0.5, scale=2}))
  n36 = Time() {bOverride_Period=true, period=600}
  Multiply(A=@n36) {constB=0.12}
  n37 = Add(A=TextureCoordinate() {coordinateIndex=1, uTiling=2.5, vTiling=2.5}, B=Constant2Vector() {g=-1.3, r=-0.4})
  n38 = Constant2Vector() {g=0.1}
  n39 = Add(A=TextureCoordinate() {coordinateIndex=1, uTiling=2.5, vTiling=2.5}, B=Constant2Vector() {g=-1.25, r=-1.05})
  n40 = Constant2Vector() {g=0.1}
  n41 = Max(
   A=TextureSample(UVs=Add(A=@n37, B=@n38)) {texture=&T_Sprinkle04_RGB}.R
   B=TextureSample(UVs=Add(A=@n39, B=@n40)) {texture=&T_Sprinkle04_RGB}.R
  )
  n42 = Add(A=Multiply(A=@n19, B=Constant2Vector() {g=3, r=1.5}), B=Constant2Vector() {g=-1.35, r=-0.2})
  n43 = TextureSample(UVs=Add(A=@n42, B=Constant2Vector() {g=0.1})) {texture=&T_Sprinkle04_RGB}
  n44 = TextureCoordinate() {coordinateIndex=1}
  n45 = Add(A=@n44, B=Constant2Vector() {g=-1, r=-0.675})
  n46 = Add(B=NamedRerouteUsage()) {constA=-0.75, constB=-0.75}
  n47 = Subtract(
   A=Multiply(A=Panner(Coordinate=@n45, Time=@n46) {bFractionalPart=true, speedY=0.25}) {constB=-3}
   B=Multiply(A=ComponentMask(None=@n44) {g=true}) {constB=0.3}
  )
  n48 = TextureSample(UVs=Add(A=@n47, B=Constant2Vector() {g=0.1})) {texture=&T_Sprinkle04_RGB}
  n49 = TextureCoordinate() {coordinateIndex=1}
  n50 = Add(A=@n49, B=Constant2Vector() {g=-1, r=-0.67})
  n51 = Add(A=NamedRerouteUsage()) {constB=2.25}
  n52 = Subtract(
   A=Multiply(A=Panner(Coordinate=@n50, Time=@n51) {bFractionalPart=true, speedY=0.25}) {constB=-3}
   B=Multiply(A=ComponentMask(None=@n49) {g=true}) {constB=0.3}
  )
  n53 = TextureSample(UVs=Add(A=@n52, B=Constant2Vector() {g=0.1})) {texture=&T_Sprinkle04_RGB}
  n54 = TextureCoordinate() {coordinateIndex=1}
  n55 = Multiply(A=ComponentMask(None=@n54) {g=true}) {constB=0.5}
  n56 = Constant2Vector() {g=-1, r=-0.675}
  n57 = NamedRerouteUsage()
  n58 = Multiply(A=Panner(Coordinate=Add(A=@n54, B=@n56), Time=Add(B=@n57) {constA=3, constB=-0.75}) {bFractionalPart=true, speedY=0.25}) {constB=-3}
  n59 = Add(A=@n55, B=@n58)
  n60 = TextureSample(UVs=Add(A=@n59, B=Constant2Vector() {g=0.1})) {texture=&T_Sprinkle04_RGB}
  n61 = TextureCoordinate() {coordinateIndex=1}
  n62 = Add(A=@n61, B=Constant2Vector() {g=-1, r=-0.66})
  n63 = Add(A=NamedRerouteUsage()) {constA=2.25, constB=0.5}
  n64 = Subtract(
   A=Multiply(A=Panner(Coordinate=@n62, Time=@n63) {bFractionalPart=true, speedY=0.25}) {constB=-3}
   B=Multiply(A=ComponentMask(None=@n61) {g=true}) {constB=0.3}
  )
  n65 = TextureSample(UVs=Add(A=@n64, B=Constant2Vector() {g=0.1})) {texture=&T_Sprinkle04_RGB}
  n66 = TextureCoordinate() {coordinateIndex=1}
  n67 = ComponentMask(None=@n66) {g=true}
  n68 = Multiply(A=@n67) {constB=0.5}
  n69 = Constant2Vector() {g=-1, r=-0.67}
  n70 = Multiply(A=Panner(Coordinate=Add(A=@n66, B=@n69), Time=NamedRerouteUsage()) {bFractionalPart=true, speedY=0.25}) {constB=-3}
  n71 = Add(A=@n68, B=@n70)
  n72 = Add(A=@n71, B=Constant2Vector() {g=0.1})
  n73 = TextureCoordinate() {coordinateIndex=1}
  n74 = Add(A=@n73, B=Constant2Vector() {g=-1, r=-0.66})
  n75 = Add(A=NamedRerouteUsage()) {constA=2.25, constB=1.4}
  n76 = Add(
   A=Multiply(A=ComponentMask(None=@n73) {g=true}) {constB=0.5}
   B=Multiply(A=Panner(Coordinate=@n74, Time=@n75) {bFractionalPart=true, speedY=0.25}) {constB=-3}
  )
  n77 = Add(A=@n76, B=Constant2Vector() {g=0.1})
  n78 = Max(
   A=Max(A=Max(A=@n48.R, B=@n53.R) {constB=0}, B=Max(A=@n60.R, B=@n65.R) {constB=0})
   B=Max(
    A=TextureSample(UVs=@n72) {texture=&T_Sprinkle04_RGB}.R
    B=TextureSample(UVs=@n77) {texture=&T_Sprinkle04_RGB}.R
   ) {constB=0}
  )
  n79 = ConstantBiasScale(None=@n67) {bias=-0.6, scale=1}
  n80 = Saturate(None=Subtract(A=@n11.A, B=Ceil(None=@n79)))
  NamedRerouteDeclaration(None=Max(A=Max(A=@n41, B=@n43.R), B=Multiply(A=@n78, B=@n80))) {name=burgertrails, nodeColor=<rgba 0.1953 0 1 1>}
  n81 = TextureSample(UVs=Rotator(Coordinate=@n37) {speed=-1.5}) {texture=&/T_Sprinkle04_Rib}
  n82 = TextureSample(UVs=Rotator(Coordinate=@n39) {speed=2}) {texture=&T_Sprinkle04_Rib}
  n83 = Max(A=@n81.RGB, B=@n82.RGB)
  n84 = TextureSample(UVs=Rotator(Coordinate=@n42) {speed=-1.5}) {texture=&T_Sprinkle04_Rib}
  n85 = TextureSample(UVs=@n47) {texture=&T_Sprinkle04_Rib}
  n86 = TextureSample(UVs=Rotator(Coordinate=@n52, Time=Constant() {r=-0.8}) {speed=2}) {texture=&T_Sprinkle04_Rib}
  n87 = TextureSample(UVs=@n59) {texture=&T_Sprinkle04_Rib}
  n88 = TextureSample(UVs=Rotator(Coordinate=@n64, Time=Constant() {r=-2.6}) {speed=2}) {texture=&T_Sprinkle04_Rib}
  n89 = TextureSample(UVs=Rotator(Coordinate=@n71, Time=Constant() {r=2}) {speed=2}) {texture=&T_Sprinkle04_Rib}
  n90 = TextureSample(UVs=Rotator(Coordinate=@n76, Time=Constant() {r=1}) {speed=2}) {texture=&T_Sprinkle04_Rib}
  n91 = Max(A=Max(A=Max(A=@n85.RGB, B=@n86.RGB) {constB=0}, B=Max(A=@n87.RGB, B=@n88.RGB) {constB=0}), B=Max(A=@n89.RGB, B=@n90.RGB) {constB=0})
  NamedRerouteDeclaration(None=Max(A=Max(A=@n83, B=@n84.RGB), B=LinearInterpolate(B=@n91, Alpha=@n80) {constB=0})) {name=burgersRGB, nodeColor=<rgba 0.125 1 0 1>}
  n92 = Max(A=@n81.A, B=@n82.A)
  n93 = LinearInterpolate(B=@n84.A, Alpha=@n11.A)
  n94 = Max(A=Max(A=Max(A=@n85.A, B=@n86.A) {constB=0}, B=Max(A=@n87.A, B=@n88.A) {constB=0}), B=Max(A=@n89.A, B=@n90.A) {constB=0})
  NamedRerouteDeclaration(None=Max(A=Max(A=@n92, B=@n93), B=Multiply(A=@n80, B=@n94))) {name=burgersA, nodeColor=<rgba 1 0 0.4453 1>}
  NamedRerouteDeclaration(None=@n36) {name=Speed, nodeColor=<rgba 0 1 0.5078 1>}
  Rotator(Coordinate=@n47, Time=Constant()) {speed=2}
  Rotator(Coordinate=@n59, Time=Constant() {r=0.1}) {speed=2}
  Saturate(None=Multiply(A=Desaturation(None=@n1.RGB)) {constB=1.75})
  Subtract(A=@n58, B=@n55)
  Subtract(A=@n70, B=@n68)
 ]}
