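# Skeletal mesh material for skin 04 of the Sprinkle vehicle body, using Default Lit shading and material attributes.
# The skin-specific function MF_Skin_Sprinkle04 supplies the look, then passes through MF_BodyNormalLOD and MF_Vehicle_FNPostFX, keeping all surface logic in material functions.
# A separate chain feeds MF_BaseVehicle_Damage into MF_VehicleDeRez_Transition and does not connect to the skin chain.
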Material /VehicleCosmetics/Bodies/Sprinkle/Skins/Mat_Vehicle_Skin_Sprinkle04
graph {
 properties {
  bAutomaticallySetUsageInEditor=false
  bUseMaterialAttributes=true
  bUsedWithMeshDeformer=true
  bUsedWithSkeletalMesh=true
  usedShadingModels=Default Lit
 }
 nodes [
  MaterialFunctionCall!(In=MaterialFunctionCall!(In=MaterialFunctionCall() {materialFunction=&/MF_Skin_Sprinkle04}.Result) {materialFunction=/VehicleCosmetics/Materials/MaterialFunctions/MF_BodyNormalLOD}.Result) {materialFunction=/VehicleCosmetics/Materials/MaterialFunctions/MF_Vehicle_FNPostFX}
  MaterialFunctionCall(In=MaterialFunctionCall() {materialFunction=&MF_BaseVehicle_Damage}.Result) {materialFunction=&MF_VehicleDeRez_Transition}
 ]
}
